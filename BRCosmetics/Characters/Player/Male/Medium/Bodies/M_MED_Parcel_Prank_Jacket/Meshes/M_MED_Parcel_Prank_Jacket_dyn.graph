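# Physics asset for the male medium Parcel Prank Jacket body cosmetic, simulating its dangling coat and waistcoat pieces.
# Five-segment left and right coat tails hang from pelvis, while single-bone front coat and waistcoat flaps hang from spine_01.
# Each piece joins through a tiny MSTR anchor capsule with free rotation, then tight twist and asymmetric swing limits.
# Capsules on pelvis, spine, legs and feet, plus a flat root box, supply body and floor shapes.
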PhysicsAsset /BRCosmetics/Characters/Player/Male/Medium/Bodies/M_MED_Parcel_Prank_Jacket/Meshes/M_MED_Parcel_Prank_Jacket_dyn
bodies {
 calf_l [{shapeType=Capsule, length=36, radius=4, center=<-19.9088 -1.8876 -0.0001>, rotation=<rot 90 -12.9074 -7.9075>}]
 calf_r [{shapeType=Capsule, length=35.75, radius=4.25, center=<19.892 1.7116 0.0001>, rotation=<rot 90 -12.9074 -7.9075>}]
 dyn_coatTail_l_01 [{shapeType=Capsule, length=5.75, radius=6.25, center=<5 0 2>, rotation=<rot 0 0 90>}]
 dyn_coatTail_l_02 [{shapeType=Capsule, length=6.5, radius=6, center=<4 0 2>, rotation=<rot 90 0 0>}]
 dyn_coatTail_l_03 [{shapeType=Capsule, length=7.25, radius=4.75, center=<3 0 2>, rotation=<rot 90 0 0>}]
 dyn_coatTail_l_04 [{shapeType=Capsule, length=3, radius=4.25, center=<3 0 1.5>, rotation=<rot 90 0 0>}]
 dyn_coatTail_l_05 [{shapeType=Capsule, length=4, radius=3.25, center=<3 0 1.5>, rotation=<rot 90 0 0>}]
 dyn_coatTail_l_MSTR [{shapeType=Capsule, length=0.5, radius=0.5}]
 dyn_coatTail_r_01 [{shapeType=Capsule, length=5.75, radius=6.25, center=<5 0 2>, rotation=<rot 0 0 90>}]
 dyn_coatTail_r_02 [{shapeType=Capsule, length=6.5, radius=6, center=<4 0 2>, rotation=<rot 90 0 0>}]
 dyn_coatTail_r_03 [{shapeType=Capsule, length=7.25, radius=4.75, center=<3 0 2>, rotation=<rot 90 0 0>}]
 dyn_coatTail_r_04 [{shapeType=Capsule, length=3, radius=4.25, center=<3 0 1.5>, rotation=<rot 90 0 0>}]
 dyn_coatTail_r_05 [{shapeType=Capsule, length=4, radius=3.25, center=<3 0 1.5>, rotation=<rot 90 0 0>}]
 dyn_coatTail_r_MSTR [{shapeType=Capsule, length=0.5, radius=0.5}]
 dyn_coat_ft_MSTR [{shapeType=Capsule, length=0.5, radius=0.5}]
 dyn_coat_ft_l [{shapeType=Capsule, length=4.25, radius=3.5, center=<8 0 2.5>, rotation=<rot 0 -30 90>}]
 dyn_coat_ft_l_MSTR [{shapeType=Capsule, length=0.5, radius=0.5}]
 dyn_coat_ft_r [{shapeType=Capsule, length=6.5, radius=3.5, center=<8 0 2.5>, rotation=<rot 0 30 90>}]
 dyn_waistcoat_l [{shapeType=Capsule, length=8.2197, radius=4.0303, center=<5 -2 2.5>, rotation=<rot 0 -30 89.9998>}]
 dyn_waistcoat_l_MSTR [{shapeType=Capsule, length=0.5, radius=0.5}]
 dyn_waistcoat_r [{shapeType=Capsule, length=8, radius=4.25, center=<5 0 2.8>, rotation=<rot 0 29.9992 89.9999>}]
 dyn_waistcoat_r_MSTR [{shapeType=Capsule, length=0.5, radius=0.5}]
 foot_l [{shapeType=Capsule, length=18.25, radius=5.5, center=<-5.3326 8.3147 0.0002>, rotation=<rot 0 0 90>}]
 foot_r [{shapeType=Capsule, length=18.25, radius=5.5, center=<5.3326 -8.3147 -0.0002>, rotation=<rot 0 0 -90>}]
 pelvis [{shapeType=Capsule, length=10.25, radius=8.5}]
 root [{shapeType=Box, extentX=150, extentY=150, extentZ=3}]
 spine_01 [{shapeType=Capsule, length=8.75, radius=8.25}]
 thigh_l [{shapeType=Capsule, length=29.75, radius=7.75, center=<-19.5662 -0.512 -0.7199>, rotation=<rot 85.5817 175.6316 169.6804>}]
 thigh_r [{shapeType=Capsule, length=32.5, radius=7.75, center=<18.8647 0.2658 0.7925>, rotation=<rot 85.5768 175.6318 169.6287>}]
}
constraints [
 {bone1Name=spine_01, bone2Name=pelvis, swing1LimitDegrees=45, swing2LimitDegrees=45, twistLimitDegrees=45}
 {bone1Name=foot_l, bone2Name=calf_l, swing1LimitDegrees=45, swing2LimitDegrees=45, twistLimitDegrees=45}
 {
  bone1Name=dyn_waistcoat_l_MSTR
  bone2Name=spine_01
  swing1LimitDegrees=45
  swing1Motion=Free
  swing2LimitDegrees=45
  swing2Motion=Free
  twistLimitDegrees=45
  twistMotion=Free
 }
 {bone1Name=dyn_waistcoat_l, bone2Name=dyn_waistcoat_l_MSTR, swing1LimitDegrees=10, swing2LimitDegrees=60, twistLimitDegrees=5}
 {
  bone1Name=dyn_waistcoat_r_MSTR
  bone2Name=spine_01
  swing1LimitDegrees=45
  swing1Motion=Free
  swing2LimitDegrees=45
  swing2Motion=Free
  twistLimitDegrees=45
  twistMotion=Free
 }
 {bone1Name=dyn_waistcoat_r, bone2Name=dyn_waistcoat_r_MSTR, swing1LimitDegrees=10, swing2LimitDegrees=60, twistLimitDegrees=5}
 {
  bone1Name=thigh_l
  bone2Name=pelvis
  swing1LimitDegrees=45
  swing1Motion=Free
  swing2LimitDegrees=45
  swing2Motion=Free
  twistLimitDegrees=45
  twistMotion=Free
 }
 {
  bone1Name=thigh_r
  bone2Name=pelvis
  swing1LimitDegrees=45
  swing1Motion=Free
  swing2LimitDegrees=45
  swing2Motion=Free
  twistLimitDegrees=45
  twistMotion=Free
 }
 {
  bone1Name=dyn_coat_ft_l_MSTR
  bone2Name=spine_01
  swing1LimitDegrees=45
  swing1Motion=Free
  swing2LimitDegrees=45
  swing2Motion=Free
  twistLimitDegrees=45
  twistMotion=Free
 }
 {bone1Name=dyn_coat_ft_l, bone2Name=dyn_coat_ft_l_MSTR, swing1LimitDegrees=10, swing2LimitDegrees=60, twistLimitDegrees=5}
 {
  bone1Name=dyn_coat_ft_MSTR
  bone2Name=spine_01
  swing1LimitDegrees=45
  swing1Motion=Free
  swing2LimitDegrees=45
  swing2Motion=Free
  twistLimitDegrees=45
  twistMotion=Free
 }
 {bone1Name=dyn_coat_ft_r, bone2Name=dyn_coat_ft_MSTR, swing1LimitDegrees=10, swing2LimitDegrees=60, twistLimitDegrees=5}
 {
  bone1Name=dyn_coatTail_l_MSTR
  bone2Name=pelvis
  swing1LimitDegrees=45
  swing1Motion=Free
  swing2LimitDegrees=45
  swing2Motion=Free
  twistLimitDegrees=45
  twistMotion=Free
 }
 {bone1Name=dyn_coatTail_l_01, bone2Name=dyn_coatTail_l_MSTR, swing1LimitDegrees=15, swing2LimitDegrees=60, twistLimitDegrees=2}
 {bone1Name=dyn_coatTail_l_02, bone2Name=dyn_coatTail_l_01, swing1LimitDegrees=20, swing2LimitDegrees=60, twistLimitDegrees=2}
 {bone1Name=dyn_coatTail_l_03, bone2Name=dyn_coatTail_l_02, swing1LimitDegrees=20, swing2LimitDegrees=45, twistLimitDegrees=2}
 {bone1Name=dyn_coatTail_l_04, bone2Name=dyn_coatTail_l_03, swing1LimitDegrees=20, swing2LimitDegrees=45, twistLimitDegrees=2}
 {bone1Name=dyn_coatTail_l_05, bone2Name=dyn_coatTail_l_04, swing1LimitDegrees=20, swing2LimitDegrees=45, twistLimitDegrees=2}
 {
  bone1Name=dyn_coatTail_r_MSTR
  bone2Name=pelvis
  swing1LimitDegrees=45
  swing1Motion=Free
  swing2LimitDegrees=45
  swing2Motion=Free
  twistLimitDegrees=45
  twistMotion=Free
 }
 {bone1Name=dyn_coatTail_r_01, bone2Name=dyn_coatTail_r_MSTR, swing1LimitDegrees=15, swing2LimitDegrees=60, twistLimitDegrees=2}
 {bone1Name=dyn_coatTail_r_02, bone2Name=dyn_coatTail_r_01, swing1LimitDegrees=20, swing2LimitDegrees=60, twistLimitDegrees=2}
 {bone1Name=dyn_coatTail_r_03, bone2Name=dyn_coatTail_r_02, swing1LimitDegrees=20, swing2LimitDegrees=45, twistLimitDegrees=2}
 {bone1Name=dyn_coatTail_r_04, bone2Name=dyn_coatTail_r_03, swing1LimitDegrees=20, swing2LimitDegrees=45, twistLimitDegrees=2}
 {bone1Name=dyn_coatTail_r_05, bone2Name=dyn_coatTail_r_04, swing1LimitDegrees=20, swing2LimitDegrees=45, twistLimitDegrees=2}
 {
  bone1Name=calf_r
  bone2Name=thigh_r
  swing1LimitDegrees=45
  swing1Motion=Free
  swing2LimitDegrees=45
  swing2Motion=Free
  twistLimitDegrees=45
  twistMotion=Free
 }
 {
  bone1Name=calf_l
  bone2Name=thigh_l
  swing1LimitDegrees=45
  swing1Motion=Free
  swing2LimitDegrees=45
  swing2Motion=Free
  twistLimitDegrees=45
  twistMotion=Free
 }
 {
  bone1Name=pelvis
  bone2Name=root
  swing1LimitDegrees=45
  swing1Motion=Free
  swing2LimitDegrees=45
  swing2Motion=Free
  twistLimitDegrees=45
  twistMotion=Free
 }
 {bone1Name=foot_r, bone2Name=calf_r, swing1LimitDegrees=45, swing2LimitDegrees=45, twistLimitDegrees=45}
]
