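# Physics asset for the dynamic collar and flower on the male medium Parcel Prank jacket.
# Nine box collar segments and a capsule flower hang from small MSTR bodies with free angular constraints to spine_05.
# Collar joints are hinge-like, wide on one swing axis, while the flower has a moderate cone; only left segments 03 and 04 are cross-linked.
# Head, clavicle and upper-arm capsules supply body collision shapes.
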PhysicsAsset /BRCosmetics/Characters/Player/Male/Medium/Bodies/M_MED_Parcel_Prank_Jacket/Meshes/M_MED_Parcel_Prank_Jacket_dyn_collar
bodies {
 clavicle_l [{shapeType=Capsule, length=4.25, radius=5.5, center=<13.1559 -0.0001 -1.4991>, rotation=<rot 79.9997 0 0>}]
 clavicle_r [{shapeType=Capsule, length=6, radius=6, center=<-12.0229 0.0001 -0.0558>, rotation=<rot 79.9997 0 -180>}]
 dyn_collar_bk [{shapeType=Box, extentX=5, extentY=13.5, extentZ=1.5, center=<2.5 0 0>}]
 dyn_collar_bk_MSTR [{shapeType=Capsule, length=0.5, radius=0.5}]
 dyn_collar_l_01 [{shapeType=Box, extentX=5, extentY=12.5, extentZ=1.5, center=<2.5 -0.0001 0.6231>, rotation=<rot -0.0001 0 -2.4067>}]
 dyn_collar_l_01_MSTR [{shapeType=Capsule, length=0.5, radius=0.5}]
 dyn_collar_l_02 [{shapeType=Box, extentX=5.5, extentY=9, extentZ=1.5, center=<3.5605 -0.793 0.4045>, rotation=<rot -8.8219 5.4997 -4.4932>}]
 dyn_collar_l_02_MSTR [{shapeType=Capsule, length=0.5, radius=0.5}]
 dyn_collar_l_03 [{shapeType=Box, extentX=6.5, extentY=9.5, extentZ=1.5, center=<4.2636 0.1521 1.0405>, rotation=<rot -9.0191 13.727 6.2859>}]
 dyn_collar_l_03_MSTR [{shapeType=Capsule, length=0.5, radius=0.5}]
 dyn_collar_l_04 [{shapeType=Box, extentX=4, extentY=7.5, extentZ=1.5, center=<1.463 0.5 0.5887>, rotation=<rot 0 20.0003 -5>}]
 dyn_collar_l_04_MSTR [{shapeType=Capsule, length=0.5, radius=0.5}]
 dyn_collar_r_01 [{shapeType=Box, extentX=5, extentY=12.5, extentZ=1.5, center=<2.5 0.0229 0.2611>, rotation=<rot 0 0 5.0002>}]
 dyn_collar_r_01_MSTR [{shapeType=Capsule, length=0.5, radius=0.5}]
 dyn_collar_r_02 [{shapeType=Box, extentX=5.5, extentY=9, extentZ=1.5, center=<3.4104 0.4354 0.2007>, rotation=<rot -9.5266 -5.8577 5.0514>}]
 dyn_collar_r_02_MSTR [{shapeType=Capsule, length=0.5, radius=0.5}]
 dyn_collar_r_03 [{shapeType=Box, extentX=6, extentY=9.5, extentZ=1.5, center=<4.1574 -0.0646 0.5904>, rotation=<rot -4.9809 -25.016 -4.5634>}]
 dyn_collar_r_03_MSTR [{shapeType=Capsule, length=0.5, radius=0.5}]
 dyn_collar_r_04 [{shapeType=Box, extentX=5, extentY=11, extentZ=1.5, center=<1.463 -0.5019 0.5887>, rotation=<rot 0 -20.0002 5.0006>}]
 dyn_collar_r_04_MSTR [{shapeType=Capsule, length=0.5, radius=0.5}]
 dyn_flower_01 [{shapeType=Capsule, length=1.25, radius=3.2207, center=<1.5 0 0>, rotation=<rot 90 -1.1935 -6.1935>}]
 dyn_flower_MSTR [{shapeType=Capsule, length=0.5, radius=0.5}]
 head [
  {shapeType=Capsule, length=1.5, radius=8.255, center=<6.4013 1.0522 0>, rotation=<rot 90 34.5085 44.5091>}
  {shapeType=Capsule, length=5.75, radius=4.005, center=<-1.3083 5.6142 0>, rotation=<rot 90 90 135.0045>}
 ]
 spine_05 [{shapeType=Capsule, length=0.5, radius=0.5}]
 upperarm_l [{shapeType=Capsule, length=17, radius=4.5, center=<8.6975 0 0>, rotation=<rot 90 0 0>}]
 upperarm_r [{shapeType=Capsule, length=17.5, radius=4, center=<-8.6975 0 0>, rotation=<rot 90 0 0>}]
}
constraints [
 {
  bone1Name=dyn_collar_bk_MSTR
  bone2Name=spine_05
  swing1LimitDegrees=45
  swing1Motion=Free
  swing2LimitDegrees=45
  swing2Motion=Free
  twistLimitDegrees=45
  twistMotion=Free
 }
 {bone1Name=dyn_collar_bk, bone2Name=dyn_collar_bk_MSTR, swing1LimitDegrees=5, swing2LimitDegrees=80, twistLimitDegrees=2}
 {
  bone1Name=dyn_collar_r_01_MSTR
  bone2Name=spine_05
  swing1LimitDegrees=45
  swing1Motion=Free
  swing2LimitDegrees=45
  swing2Motion=Free
  twistLimitDegrees=45
  twistMotion=Free
 }
 {bone1Name=dyn_collar_r_01, bone2Name=dyn_collar_r_01_MSTR, swing1LimitDegrees=5, swing2LimitDegrees=80, twistLimitDegrees=2}
 {
  bone1Name=dyn_collar_r_02_MSTR
  bone2Name=spine_05
  swing1LimitDegrees=45
  swing1Motion=Free
  swing2LimitDegrees=45
  swing2Motion=Free
  twistLimitDegrees=45
  twistMotion=Free
 }
 {bone1Name=dyn_collar_r_02, bone2Name=dyn_collar_r_02_MSTR, swing1LimitDegrees=5, swing2LimitDegrees=80, twistLimitDegrees=2}
 {
  bone1Name=dyn_collar_r_03_MSTR
  bone2Name=spine_05
  swing1LimitDegrees=45
  swing1Motion=Free
  swing2LimitDegrees=45
  swing2Motion=Free
  twistLimitDegrees=45
  twistMotion=Free
 }
 {bone1Name=dyn_collar_r_03, bone2Name=dyn_collar_r_03_MSTR, swing1LimitDegrees=5, swing2LimitDegrees=80, twistLimitDegrees=2}
 {
  bone1Name=dyn_collar_r_04_MSTR
  bone2Name=spine_05
  swing1LimitDegrees=45
  swing1Motion=Free
  swing2LimitDegrees=45
  swing2Motion=Free
  twistLimitDegrees=45
  twistMotion=Free
 }
 {bone1Name=dyn_collar_r_04, bone2Name=dyn_collar_r_04_MSTR, swing1LimitDegrees=5, swing2LimitDegrees=80, twistLimitDegrees=2}
 {
  bone1Name=dyn_collar_l_01_MSTR
  bone2Name=spine_05
  swing1LimitDegrees=45
  swing1Motion=Free
  swing2LimitDegrees=45
  swing2Motion=Free
  twistLimitDegrees=45
  twistMotion=Free
 }
 {bone1Name=dyn_collar_l_01, bone2Name=dyn_collar_l_01_MSTR, swing1LimitDegrees=5, swing2LimitDegrees=80, twistLimitDegrees=2}
 {
  bone1Name=dyn_collar_l_02_MSTR
  bone2Name=spine_05
  swing1LimitDegrees=45
  swing1Motion=Free
  swing2LimitDegrees=45
  swing2Motion=Free
  twistLimitDegrees=45
  twistMotion=Free
 }
 {bone1Name=dyn_collar_l_02, bone2Name=dyn_collar_l_02_MSTR, swing1LimitDegrees=5, swing2LimitDegrees=80, twistLimitDegrees=2}
 {
  bone1Name=dyn_collar_l_03_MSTR
  bone2Name=spine_05
  swing1LimitDegrees=45
  swing1Motion=Free
  swing2LimitDegrees=45
  swing2Motion=Free
  twistLimitDegrees=45
  twistMotion=Free
 }
 {bone1Name=dyn_collar_l_03, bone2Name=dyn_collar_l_03_MSTR, swing1LimitDegrees=5, swing2LimitDegrees=80, twistLimitDegrees=2}
 {
  bone1Name=dyn_collar_l_04_MSTR
  bone2Name=spine_05
  swing1LimitDegrees=45
  swing1Motion=Free
  swing2LimitDegrees=45
  swing2Motion=Free
  twistLimitDegrees=45
  twistMotion=Free
 }
 {bone1Name=dyn_collar_l_04, bone2Name=dyn_collar_l_04_MSTR, swing1LimitDegrees=5, swing2LimitDegrees=80, twistLimitDegrees=2}
 {
  bone1Name=upperarm_r
  bone2Name=clavicle_r
  swing1LimitDegrees=45
  swing1Motion=Free
  swing2LimitDegrees=45
  swing2Motion=Free
  twistLimitDegrees=45
  twistMotion=Free
 }
 {
  bone1Name=clavicle_l
  bone2Name=spine_05
  swing1LimitDegrees=45
  swing1Motion=Free
  swing2LimitDegrees=45
  swing2Motion=Free
  twistLimitDegrees=45
  twistMotion=Free
 }
 {
  bone1Name=upperarm_l
  bone2Name=clavicle_l
  swing1LimitDegrees=45
  swing1Motion=Free
  swing2LimitDegrees=45
  swing2Motion=Free
  twistLimitDegrees=45
  twistMotion=Free
 }
 {
  bone1Name=dyn_collar_l_04
  bone2Name=dyn_collar_l_03
  swing1LimitDegrees=45
  swing1Motion=Free
  swing2LimitDegrees=15
  twistLimitDegrees=45
  twistMotion=Free
 }
 {
  bone1Name=dyn_flower_MSTR
  bone2Name=spine_05
  swing1LimitDegrees=45
  swing1Motion=Free
  swing2LimitDegrees=45
  swing2Motion=Free
  twistLimitDegrees=45
  twistMotion=Free
 }
 {bone1Name=dyn_flower_01, bone2Name=dyn_flower_MSTR, swing1LimitDegrees=30, swing2LimitDegrees=30, twistLimitDegrees=5}
 {
  bone1Name=head
  bone2Name=spine_05
  swing1LimitDegrees=45
  swing1Motion=Free
  swing2LimitDegrees=45
  swing2Motion=Free
  twistLimitDegrees=45
  twistMotion=Free
 }
 {
  bone1Name=clavicle_r
  bone2Name=spine_05
  swing1LimitDegrees=45
  swing1Motion=Free
  swing2LimitDegrees=45
  swing2Motion=Free
  twistLimitDegrees=45
  twistMotion=Free
 }
]
