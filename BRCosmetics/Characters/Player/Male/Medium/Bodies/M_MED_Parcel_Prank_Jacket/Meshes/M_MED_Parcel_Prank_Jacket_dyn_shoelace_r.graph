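# Physics asset for the dynamic right shoelace on the male medium Parcel Prank Jacket character body.
# Two shoelace bones each carry a small box body, alongside a rotated capsule body on the right foot.
# Both lace bodies are constrained directly to the foot rather than chained to each other, with moderate swing limits and a very tight twist limit.
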PhysicsAsset /BRCosmetics/Characters/Player/Male/Medium/Bodies/M_MED_Parcel_Prank_Jacket/Meshes/M_MED_Parcel_Prank_Jacket_dyn_shoelace_r
bodies {
 dyn_shoelace_01_r [{shapeType=Box, extentX=2, extentY=2, extentZ=1.2, center=<2 0 0>}]
 dyn_shoelace_02_r [{shapeType=Box, extentX=2, extentY=2, extentZ=1.2, center=<2.6108 0 0.0249>}]
 foot_r [{shapeType=Capsule, length=8.45, radius=3.55, center=<4.4716 -4.6391 0.4315>, rotation=<rot 90 26.565 94.014>}]
}
constraints [
 {bone1Name=dyn_shoelace_01_r, bone2Name=foot_r, swing1LimitDegrees=30, swing2LimitDegrees=45, twistLimitDegrees=2}
 {bone1Name=dyn_shoelace_02_r, bone2Name=foot_r, swing1LimitDegrees=30, swing2LimitDegrees=45, twistLimitDegrees=2}
]
