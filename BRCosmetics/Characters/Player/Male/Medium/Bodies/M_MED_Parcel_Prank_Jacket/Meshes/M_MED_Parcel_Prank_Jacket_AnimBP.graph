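# Post-process style cosmetic AnimBlueprint for the male medium Parcel Prank Jacket, adding secondary motion to the pose copied from its attached parent mesh.
# Constraints drive collar and flower, and AnimDynamics swings a chain.
# With rigid bodies enabled, thigh- and crouch-driven corrections feed jacket, collar and shoelace simulations; otherwise only waistcoat corrections apply.
# The event graph computes thigh lift and gravity override, and resets dynamics around skydiving and parachuting.
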AnimBlueprint /BRCosmetics/Characters/Player/Male/Medium/Bodies/M_MED_Parcel_Prank_Jacket/Meshes/M_MED_Parcel_Prank_Jacket_AnimBP
target_skeleton=&/M_MED_Parcel_Prank_Jacket_Skeleton
graphs [
 AnimGraph {nodes [
   "Constraint - dyn_chain_b_MSTR (PPPP)"!(
    ComponentPose
    Alpha=1
    ConstraintWeights_0=0.4024
    ConstraintWeights_1:float=0.0853
    ConstraintWeights_2:float=0.02
    ConstraintWeights_3:float=0.4923
   ) -> (Pose) {
    alphaBoolBlend {blendOption=Linear}
    boneToModify {boneName=dyn_chain_b_MSTR}
    constraintSetup [{targetBone {boneName=spine_01}}, {targetBone {boneName=spine_02}}, {targetBone {boneName=spine_03}}, {targetBone {boneName=pelvis}}]
    constraintWeights [1,1,1,1]
   }
   n1 = "Sequence Player (additive) 'M_MED_Parcel_Prank_Jacket_Reset_Pose'"() {sequence=&/M_MED_Parcel_Prank_Jacket_Reset_Pose}
   n2 = Use cached pose 'PerPose'()
   RigidBody!(
    ComponentPose=Local To Component(LocalPose=Apply Additive(Base=@n2, Additive=@n1) {alphaBoolBlend {blendOption=Linear}})
    Alpha
    OverrideWorldGravity:Vector="0,0,0"
    ExternalForce
   ) -> (Pose) {
    bClampLinearTranslationLimitToRefPose=true
    bOverrideWorldGravity=true
    overridePhysicsAsset=&/M_MED_Parcel_Prank_Jacket_dyn
    simulationSpace=BaseBoneSpace
    alphaBoolBlend {blendOption=Linear}
    baseBoneRef {boneName=pelvis}
    componentAppliedLinearAccClamp=<1000 1000 1000>
    componentLinearAccScale=<0.2 0.25 0.2>
    componentLinearVelScale=<0.3 0.3 0.3>
    overrideWorldGravity {z=-500}
    simSpaceSettings {
     maxAngularAcceleration=300
     maxAngularVelocity=6
     maxLinearAcceleration=800
     maxLinearVelocity=50
     velocityScaleZ=0
     externalLinearDragV {y=1, z=1}
    }
   }
   n3 = "Transform (Modify) Bone - Bone: dyn_waistcoat_l_MSTR"(
    ComponentPose=Local To Component(LocalPose=Use cached pose 'PerPose'())
    Alpha=Get thigh_rot_up_l() -> (thigh_rot_up_l:double)
    Translation="-1.138,-0.000,-0.000"
    Rotation="0,0,0"
    Scale="0.6,1.000,1.000"
   ) {
    rotationMode=BMM_Additive
    rotationSpace=BCS_BoneSpace
    scaleMode=BMM_Additive
    scaleSpace=BCS_BoneSpace
    translationMode=BMM_Additive
    translationSpace=BCS_BoneSpace
    alphaBoolBlend {blendOption=Linear}
    boneToModify {boneName=dyn_waistcoat_l_MSTR}
    scale {x=0.6}
    translation {x=-1.138}
   }
   n4 = "Transform (Modify) Bone - Bone: dyn_waistcoat_r_MSTR"(
    ComponentPose=@n3
    Alpha=Get thigh_rot_up_r() -> (thigh_rot_up_r:double)
    Translation="-1.138,0,0"
    Rotation="0.000,5.000,0.000"
    Scale="0.6,1,1"
   ) {
    rotationMode=BMM_Additive
    rotationSpace=BCS_BoneSpace
    scaleMode=BMM_Additive
    scaleSpace=BCS_BoneSpace
    translationMode=BMM_Additive
    translationSpace=BCS_BoneSpace
    alphaBoolBlend {blendOption=Linear}
    boneToModify {boneName=dyn_waistcoat_r_MSTR}
    rotation {yaw=5.0001}
    scale {x=0.6}
    translation {x=-1.138}
   }
   Output Pose(Result=Blend Poses by bool(BlendPose_0=Use cached pose 'rigidPose'(), BlendPose_1=Component To Local(ComponentPose=@n4), bActiveValue=Get bRigidBodyEnabled()) {blendType=Linear}) {layerGroup=None}
   n5 = "Constraint - dyn_collar_bk_MSTR (PPPP)"!(
    ComponentPose=Local To Component(LocalPose=Copy Pose From Mesh() {bUseAttachedParent=true})
    Alpha=1
    ConstraintWeights_0=0.0011
    ConstraintWeights_1:float=0.4998
    ConstraintWeights_2:float=0.1268
    ConstraintWeights_3:float=0.3723
   ) -> (Pose) {
    alphaBoolBlend {blendOption=Linear}
    boneToModify {boneName=dyn_collar_bk_MSTR}
    constraintSetup [{targetBone {boneName=clavicle_r}}, {targetBone {boneName=neck_02}}, {targetBone {boneName=neck_01}}, {targetBone {boneName=spine_05}}]
    constraintWeights [1,1,1,1]
   }
   n6 = "Constraint - dyn_collar_r_01_MSTR (PPPP)"!(
    ComponentPose=@n5
    Alpha=1
    ConstraintWeights_0=0.0345
    ConstraintWeights_1:float=0.3575
    ConstraintWeights_2:float=0.1136
    ConstraintWeights_3:float=0.4944
   ) -> (Pose) {
    alphaBoolBlend {blendOption=Linear}
    boneToModify {boneName=dyn_collar_r_01_MSTR}
    constraintSetup [{targetBone {boneName=clavicle_r}}, {targetBone {boneName=neck_02}}, {targetBone {boneName=neck_01}}, {targetBone {boneName=spine_05}}]
    constraintWeights [1,1,1,1]
   }
   n7 = "Constraint - dyn_collar_r_02_MSTR (PPPP)"!(
    ComponentPose=@n6
    Alpha=1
    ConstraintWeights_0=0.081
    ConstraintWeights_1:float=0.166
    ConstraintWeights_2:float=0.0707
    ConstraintWeights_3:float=0.6824
   ) -> (Pose) {
    alphaBoolBlend {blendOption=Linear}
    boneToModify {boneName=dyn_collar_r_02_MSTR}
    constraintSetup [{targetBone {boneName=clavicle_r}}, {targetBone {boneName=neck_02}}, {targetBone {boneName=neck_01}}, {targetBone {boneName=spine_05}}]
    constraintWeights [1,1,1,1]
   }
   n8 = "Constraint - dyn_collar_r_03_MSTR (PPPP)"!(
    ComponentPose=@n7
    Alpha=1
    ConstraintWeights_0=0.1329
    ConstraintWeights_1:float=0.0447
    ConstraintWeights_2:float=0.0271
    ConstraintWeights_3:float=0.7953
   ) -> (Pose) {
    alphaBoolBlend {blendOption=Linear}
    boneToModify {boneName=dyn_collar_r_03_MSTR}
    constraintSetup [{targetBone {boneName=clavicle_r}}, {targetBone {boneName=neck_02}}, {targetBone {boneName=neck_01}}, {targetBone {boneName=spine_05}}]
    constraintWeights [1,1,1,1]
   }
   n9 = "Constraint - dyn_collar_r_04_MSTR (PPP)"!(ComponentPose=@n8, Alpha=1, ConstraintWeights_0=0.105, ConstraintWeights_1:float=0.0036, ConstraintWeights_2:float=0.8914) -> (Pose) {
    alphaBoolBlend {blendOption=Linear}
    boneToModify {boneName=dyn_collar_r_04_MSTR}
    constraintSetup [{targetBone {boneName=clavicle_r}}, {targetBone {boneName=neck_01}}, {targetBone {boneName=spine_05}}]
    constraintWeights [1,1,1]
   }
   n10 = "Constraint - dyn_collar_l_01_MSTR (PPPP)"!(
    ComponentPose=@n9
    Alpha=1
    ConstraintWeights_0=0.0345
    ConstraintWeights_1:float=0.3575
    ConstraintWeights_2:float=0.1136
    ConstraintWeights_3:float=0.4944
   ) -> (Pose) {
    alphaBoolBlend {blendOption=Linear}
    boneToModify {boneName=dyn_collar_l_01_MSTR}
    constraintSetup [{targetBone {boneName=clavicle_l}}, {targetBone {boneName=neck_02}}, {targetBone {boneName=neck_01}}, {targetBone {boneName=spine_05}}]
    constraintWeights [1,1,1,1]
   }
   n11 = "Constraint - dyn_collar_l_02_MSTR (PPPP)"!(
    ComponentPose=@n10
    Alpha=1
    ConstraintWeights_0=0.081
    ConstraintWeights_1:float=0.166
    ConstraintWeights_2:float=0.0707
    ConstraintWeights_3:float=0.6824
   ) -> (Pose) {
    alphaBoolBlend {blendOption=Linear}
    boneToModify {boneName=dyn_collar_l_02_MSTR}
    constraintSetup [{targetBone {boneName=clavicle_l}}, {targetBone {boneName=neck_02}}, {targetBone {boneName=neck_01}}, {targetBone {boneName=spine_05}}]
    constraintWeights [1,1,1,1]
   }
   n12 = "Constraint - dyn_collar_l_03_MSTR (PPPP)"!(
    ComponentPose=@n11
    Alpha=1
    ConstraintWeights_0=0.1329
    ConstraintWeights_1:float=0.0447
    ConstraintWeights_2:float=0.0271
    ConstraintWeights_3:float=0.7953
   ) -> (Pose) {
    alphaBoolBlend {blendOption=Linear}
    boneToModify {boneName=dyn_collar_l_03_MSTR}
    constraintSetup [{targetBone {boneName=clavicle_l}}, {targetBone {boneName=neck_02}}, {targetBone {boneName=neck_01}}, {targetBone {boneName=spine_05}}]
    constraintWeights [1,1,1,1]
   }
   n13 = "Constraint - dyn_collar_l_04_MSTR (PPP)"!(ComponentPose=@n12, Alpha=1, ConstraintWeights_0=0.105, ConstraintWeights_1:float=0.0036, ConstraintWeights_2:float=0.8914) -> (Pose) {
    alphaBoolBlend {blendOption=Linear}
    boneToModify {boneName=dyn_collar_l_04_MSTR}
    constraintSetup [{targetBone {boneName=clavicle_l}}, {targetBone {boneName=neck_01}}, {targetBone {boneName=spine_05}}]
    constraintWeights [1,1,1]
   }
   n14 = "Constraint - dyn_flower_MSTR (PPPPPPP)"!(
    ComponentPose=@n13
    Alpha=1
    ConstraintWeights_0=0.0267
    ConstraintWeights_1:float=0.0001
    ConstraintWeights_2:float=0.0587
    ConstraintWeights_3:float=0.0373
    ConstraintWeights_4:float=0.4531
    ConstraintWeights_5:float=0.4237
    ConstraintWeights_6:float=0.0005
   ) -> (Pose) {
    alphaBoolBlend {blendOption=Linear}
    boneToModify {boneName=dyn_flower_MSTR}
    constraintSetup [
     {targetBone {boneName=upperarm_twist_01_r}}
     {targetBone {boneName=spine_02}}
     {targetBone {boneName=spine_03}}
     {targetBone {boneName=spine_04}}
     {targetBone {boneName=spine_05}}
     {targetBone {boneName=clavicle_r}}
     {targetBone {boneName=neck_01}}
    ]
    constraintWeights [1,1,1,1,1,1,1]
   }
   n15 = Anim Dynamics - Bone: dyn_chain(ComponentPose=@n14, Alpha=1, ExternalForce="0,0,0") {
    angularDampingOverride=1
    angularSpringConstant=1500
    bAngularSpring=true
    bGravityOverrideInSimSpace=true
    bOverrideAngularDamping=true
    gravityScale=0.35
    simulationSpace=RootRelative
    alphaBoolBlend {blendOption=Linear}
    alphaScaleBiasClamp {scale=0.6}
    boundBone {boneName=dyn_chain}
    componentAppliedLinearAccClamp=<3000 3000 3000>
    componentLinearAccScale {z=0.45}
    componentLinearVelScale {z=0.45}
    physicsBodyDefinitions [{
      sphereCollisionRadius=0
      boundBone {boneName=dyn_chain}
      boxExtents=<10 20 20>
      constraintSetup {angularLimitsMax=<0 40 0>, angularLimitsMin=<0 0 0>, angularTarget=<0 0 0>}
      localJointOffset=<10 0 0>
     }]
   }
   Save cached pose 'PerPose'(Pose=Component To Local(ComponentPose=@n15)) {cacheName=PerPose}
   n16 = Get crouchingBegin() -> (crouchingBegin:double)
   n17 = "Transform (Modify) Bone - Bone: dyn_coatTail_l_MSTR"(
    ComponentPose=Local To Component(LocalPose=Use cached pose 'PerPose'())
    Alpha=@n16
    Translation="0,0,0"
    Rotation="45,0.000,0.000"
    Scale="1,1,1"
   ) {
    rotationMode=BMM_Additive
    rotationSpace=BCS_BoneSpace
    alphaBoolBlend {blendOption=Linear}
    alphaScaleBiasClamp {bInterpResult=true, interpSpeedIncreasing=30}
    boneToModify {boneName=dyn_coatTail_l_MSTR}
    rotation {pitch=45}
   }
   n18 = "Transform (Modify) Bone - Bone: dyn_coatTail_r_MSTR"(ComponentPose=@n17, Alpha=@n16, Translation="0,0,0", Rotation="45,0.000,0.000", Scale="1,1,1") {
    rotationMode=BMM_Additive
    rotationSpace=BCS_BoneSpace
    alphaBoolBlend {blendOption=Linear}
    alphaScaleBiasClamp {bInterpResult=true, interpSpeedIncreasing=30}
    boneToModify {boneName=dyn_coatTail_r_MSTR}
    rotation {pitch=45}
   }
   n19 = Get thigh_rot_up_l()
   n20 = "Transform (Modify) Bone - Bone: dyn_waistcoat_l_MSTR"(ComponentPose=@n18, Alpha=@n19, Translation="-3.949,1.452,1.982", Rotation="0,-26.172,-5", Scale="1,1,1") {
    rotationMode=BMM_Additive
    rotationSpace=BCS_BoneSpace
    translationMode=BMM_Additive
    translationSpace=BCS_BoneSpace
    alphaBoolBlend {blendOption=Linear}
    boneToModify {boneName=dyn_waistcoat_l_MSTR}
    rotation {roll=-5, yaw=-26.172}
    translation=<-3.949 1.452 1.982>
   }
   n21 = "Transform (Modify) Bone - Bone: dyn_coat_ft_l_MSTR"(ComponentPose=@n20, Alpha=@n19, Translation="-1.5,0,1.5", Rotation="0,0,0", Scale="1,1,1") {
    rotationMode=BMM_Additive
    rotationSpace=BCS_BoneSpace
    translationMode=BMM_Additive
    translationSpace=BCS_BoneSpace
    alphaBoolBlend {blendOption=Linear}
    boneToModify {boneName=dyn_coat_ft_l_MSTR}
    translation {x=-1.5, z=1.5}
   }
   n22 = Get thigh_rot_up_r()
   n23 = "Transform (Modify) Bone - Bone: dyn_waistcoat_r_MSTR"(ComponentPose=@n21, Alpha=@n22, Translation="-1.643,0.146,-0.000", Rotation="0.000,10.000,0.000", Scale="1,1,1") {
    rotationMode=BMM_Additive
    rotationSpace=BCS_BoneSpace
    translationMode=BMM_Additive
    translationSpace=BCS_BoneSpace
    alphaBoolBlend {blendOption=Linear}
    boneToModify {boneName=dyn_waistcoat_r_MSTR}
    rotation {yaw=10.0003}
    translation=<-1.643 0.146>
   }
   n24 = "Transform (Modify) Bone - Bone: dyn_coat_ft_MSTR"(ComponentPose=@n23, Alpha=@n22, Translation="-1.5,0,1", Rotation="0,0,0", Scale="1,1,1") {
    rotationMode=BMM_Additive
    rotationSpace=BCS_BoneSpace
    translationMode=BMM_Additive
    translationSpace=BCS_BoneSpace
    alphaBoolBlend {blendOption=Linear}
    boneToModify {boneName=dyn_coat_ft_MSTR}
    translation {x=-1.5, z=1}
   }
   n25 = RigidBody(ComponentPose=@n24, OverrideWorldGravity=Get gravity_override() -> (gravity_override:Vector)) {
    bClampLinearTranslationLimitToRefPose=true
    bOverrideWorldGravity=true
    overridePhysicsAsset=&M_MED_Parcel_Prank_Jacket_dyn
    simulationSpace=BaseBoneSpace
    alphaBoolBlend {blendOption=Linear}
    baseBoneRef {boneName=pelvis}
    componentAppliedLinearAccClamp=<1000 1000 1000>
    componentLinearAccScale=<0.2 0.25 0.2>
    componentLinearVelScale=<0.3 0.3 0.3>
    overrideWorldGravity {z=-500}
    simSpaceSettings {
     maxAngularAcceleration=300
     maxAngularVelocity=6
     maxLinearAcceleration=800
     maxLinearVelocity=50
     velocityScaleZ=0
     externalLinearDragV {y=1, z=1}
    }
   }
   n26 = "Transform (Modify) Bone - Bone: dyn_waistcoat_l"(ComponentPose=@n25, Alpha=Get thigh_rot_up_l(), Translation="0,0,0", Rotation="0,0,0", Scale="0.65,1.000,1") {
    rotationMode=BMM_Additive
    rotationSpace=BCS_BoneSpace
    scaleMode=BMM_Additive
    scaleSpace=BCS_BoneSpace
    translationMode=BMM_Additive
    translationSpace=BCS_BoneSpace
    alphaBoolBlend {blendOption=Linear}
    boneToModify {boneName=dyn_waistcoat_l}
    scale {x=0.65}
   }
   n27 = RigidBody!(ComponentPose=@n26, Alpha, ExternalForce) -> (Pose) {
    bClampLinearTranslationLimitToRefPose=true
    bOverrideWorldGravity=true
    overridePhysicsAsset=&/M_MED_Parcel_Prank_Jacket_dyn_collar
    simulationSpace=BaseBoneSpace
    alphaBoolBlend {blendOption=Linear}
    baseBoneRef {boneName=spine_05}
    componentAppliedLinearAccClamp=<2000 2000 2000>
    componentLinearAccScale=<0.2 0.2 0.2>
    componentLinearVelScale=<0.2 0.2 0.2>
    overrideWorldGravity {z=-300}
   }
   n28 = RigidBody(ComponentPose=@n27) {
    bClampLinearTranslationLimitToRefPose=true
    bOverrideWorldGravity=true
    overridePhysicsAsset=&/M_MED_Parcel_Prank_Jacket_dyn_shoelace_r
    simulationSpace=BaseBoneSpace
    alphaBoolBlend {blendOption=Linear}
    baseBoneRef {boneName=calf_r}
    componentAppliedLinearAccClamp=<1000 1000 1000>
    componentLinearAccScale=<0.25 0.25 0.25>
    componentLinearVelScale=<0.2 0.2 0.2>
    overrideWorldGravity {z=-500}
   }
   n29 = RigidBody(ComponentPose=@n28) {
    bClampLinearTranslationLimitToRefPose=true
    bOverrideWorldGravity=true
    overridePhysicsAsset=&/M_MED_Parcel_Prank_Jacket_dyn_shoelace_l
    simulationSpace=BaseBoneSpace
    alphaBoolBlend {blendOption=Linear}
    baseBoneRef {boneName=calf_l}
    componentAppliedLinearAccClamp=<1000 1000 1000>
    componentLinearAccScale=<0.25 0.25 0.25>
    componentLinearVelScale=<0.2 0.2 0.2>
    overrideWorldGravity {z=-500}
   }
   Save cached pose 'rigidPose'(Pose=Component To Local(ComponentPose=@n29)) {cacheName=rigidPose}
   SelectVector(A="10,0,100", B="0, 0,-980", bPickA=OR Boolean!(A=Get bIsSkydiveFloating(), B=Get bIsSkyDiving(), C:bool) -> (ReturnValue))
  ]}
 EventGraph {nodes [
   Event BlueprintUpdateAnimation()
   n1 = BreakTransform(InTransform=GetDeltaTransformFromRefPose(self=GetOwnerSkeletalMeshComponent(), BoneName=thigh_r, BaseName=pelvis))
   > Set thigh_rot_up_r(thigh_rot_up_r:double=MapRangeClamped(Value=BreakRotator(InRot=@n1.Rotation).Yaw, InRangeA=0, InRangeB=-120)) -> (Output_Get:double)
   n2 = BreakTransform(InTransform=GetDeltaTransformFromRefPose(self=GetOwnerSkeletalMeshComponent(), BoneName=thigh_l, BaseName=pelvis))
   > Set thigh_rot_up_l(thigh_rot_up_l:double=MapRangeClamped(Value=BreakRotator(InRot=@n2.Rotation).Yaw, InRangeA=-30, InRangeB=-120)) -> (Output_Get:double)
   > n3 = Gravity Override State Function DEPRECATED(CharacterPartAnimInstance=Self-Reference(), GameTimeInSeconds=GetGameTimeInSeconds(), GravityOverrideParams=Get Gravity Override Params())
   > Set gravity_override(gravity_override=@n3)
   Event OnBeginCrouching()
   > Set crouchingBegin(crouchingBegin:double=1) -> (Output_Get:double)
   Event OnEndCrouching()
   > Set crouchingBegin(crouchingBegin=0)
   Event OnParachuteOpened()
   > n4 = Delay(Duration=1.5)
   ResetDynamics(execute=[@n4.then, @n5.then, @n6.then], InTeleportType=TeleportPhysics)
   > ResetDynamics(InTeleportType=ResetPhysics)
   n5 = Event OnEndSkydiving()
   n6 = Event OnBeginSkydiving()
   ResetDynamics(InTeleportType=TeleportPhysics)
   Event BlueprintInitializeAnimation()
  ]}
]
variables [
 crouchingBegin {type=number}
 Gravity Override Params {type=GravityOverrideParamsStruct, default {skydive_FloatingDown {windAmplitude=<2000 -5000 4000>, windFrequency=<500 800 1000>}}}
 gravity_override {type=Vector}
 thigh_rot_up_l {type=number}
 thigh_rot_up_r {type=number}
]
